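# Physics asset for the Lazarus Lens style NPC medium male body, built for secondary motion of its coat and skirt.
# Capsules on pelvis, legs and feet, plus a root box, form the lower body, with free angular constraints.
# Limited-angle capsule chains hang coat panels and back skirt from the pelvis, front skirt from the thighs.
# Angularly free cross-links tie mid to front coat panels and join the two back skirt tips.
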PhysicsAsset /BRCosmetics/Characters/Player/Male/Medium/Bodies/M_MED_LazarusLens_StyleNPC/Meshes/M_MED_LazarusLens_Style_NPC_Physics
bodies {
 calf_l [{shapeType=Capsule, length=39, radius=11.5, center=<-18.002 0 0>, rotation=<rot 90 0 0>}]
 calf_r [{shapeType=Capsule, length=40, radius=11.5, center=<18.002 0 0>, rotation=<rot 90 0 0>}]
 dyn_coat_back_01_l [{shapeType=Capsule, length=12, radius=6.5, shapeName=dyn_coat_back_01_l_capsule, center=<4 0 3>, rotation=<rot 0 0 90>}]
 dyn_coat_back_01_r [{shapeType=Capsule, length=12, radius=6.5, shapeName=dyn_coat_back_01_r_capsule, center=<4 0 3>, rotation=<rot 0 0 90>}]
 dyn_coat_back_02_l [{shapeType=Capsule, length=12, radius=6.5, shapeName=dyn_coat_back_02_l_capsule, center=<4 0 3>, rotation=<rot 0 0 90>}]
 dyn_coat_back_02_r [{shapeType=Capsule, length=12, radius=6.5, shapeName=dyn_coat_back_02_r_capsule, center=<4 0 3>, rotation=<rot 0 0 90>}]
 dyn_coat_back_03_l [{shapeType=Capsule, length=12, radius=6.5, shapeName=dyn_coat_back_03_l_capsule, center=<4 0 3>, rotation=<rot 0 0 90>}]
 dyn_coat_back_03_r [{shapeType=Capsule, length=12, radius=6.5, shapeName=dyn_coat_back_03_r_capsule, center=<4 0 3>, rotation=<rot 0 0 90>}]
 dyn_coat_front_01_l [{shapeType=Capsule, length=12, radius=6.5, shapeName=dyn_coat_front_01_l_capsule, center=<4 0 4>, rotation=<rot 0 0 90>}]
 dyn_coat_front_01_r [{shapeType=Capsule, length=12, radius=6.5, shapeName=dyn_coat_front_01_r_capsule, center=<4 0 4>, rotation=<rot 0 0 90>}]
 dyn_coat_front_02_l [{shapeType=Capsule, length=18, radius=6.5, shapeName=dyn_coat_front_02_l_capsule, center=<4 0 4>, rotation=<rot 0 0 130>}]
 dyn_coat_front_02_r [{shapeType=Capsule, length=22, radius=6.5, shapeName=dyn_coat_front_02_r_capsule, center=<4 0 4>, rotation=<rot 0 0 60>}]
 dyn_coat_front_03_l [{shapeType=Capsule, length=16.5, radius=8, shapeName=dyn_coat_front_03_l_capsule, center=<4 0 4>, rotation=<rot 0 0 130>}]
 dyn_coat_front_03_r [{shapeType=Capsule, length=31, radius=7.5, shapeName=dyn_coat_front_03_r_capsule, center=<4 0 4>, rotation=<rot 0 0 60>}]
 dyn_coat_front_04_l [{shapeType=Capsule, length=26.5, radius=6.5, shapeName=dyn_coat_front_04_l_capsule, center=<4 0 4>, rotation=<rot 0 0 130>}]
 dyn_coat_front_04_r [{shapeType=Capsule, length=22, radius=6.5, shapeName=dyn_coat_front_04_r_capsule, center=<4 0 4>, rotation=<rot 0 0 60>}]
 dyn_coat_mid_01_l [{shapeType=Capsule, length=12, radius=6.5, shapeName=dyn_coat_mid_01_l_capsule, center=<4 0 5>, rotation=<rot 0 0 90>}]
 dyn_coat_mid_01_r [{shapeType=Capsule, length=12, radius=6.5, shapeName=dyn_coat_mid_01_r_capsule, center=<4 0 3>, rotation=<rot 0 0 90>}]
 dyn_coat_mid_02_l [{shapeType=Capsule, length=12, radius=6.5, shapeName=dyn_coat_mid_02_l_capsule, center=<4 0 3>, rotation=<rot 0 0 90>}]
 dyn_coat_mid_02_r [{shapeType=Capsule, length=12, radius=6.5, shapeName=dyn_coat_mid_02_r_capsule, center=<4 0 3>, rotation=<rot 0 0 90>}]
 dyn_coat_mid_03_l [{shapeType=Capsule, length=12.5, radius=6, shapeName=dyn_coat_mid_03_l_capsule, center=<4 0 3>, rotation=<rot 0 0 90>}]
 dyn_coat_mid_03_r [{shapeType=Capsule, length=18, radius=6.5, shapeName=dyn_coat_mid_03_r_capsule, center=<4 0 3>, rotation=<rot 0 0 90>}]
 dyn_skirt_back_01_l [{shapeType=Capsule, length=14, radius=4, shapeName=dyn_skirt_back_01_l_capsule, center=<4 0 0>, rotation=<rot 0 0 90>}]
 dyn_skirt_back_01_r [{shapeType=Capsule, length=14, radius=4, shapeName=dyn_skirt_back_01_r_capsule, center=<4 0 0>, rotation=<rot 0 0 90>}]
 dyn_skirt_back_02_l [{shapeType=Capsule, length=14, radius=4, shapeName=dyn_skirt_back_02_l_capsule, center=<4 0 0>, rotation=<rot 0 0 90>}]
 dyn_skirt_back_02_r [{shapeType=Capsule, length=14, radius=4, shapeName=dyn_skirt_back_02_r_capsule, center=<4 0 0>, rotation=<rot 0 0 90>}]
 dyn_skirt_back_03_l [{shapeType=Capsule, length=11.5, radius=4, shapeName=dyn_skirt_back_03_l_capsule, center=<4 0 0>, rotation=<rot 0 0 90>}]
 dyn_skirt_back_03_r [{shapeType=Capsule, length=11.5, radius=4, shapeName=dyn_skirt_back_03_r_capsule, center=<4 0 0>, rotation=<rot 0 0 90>}]
 dyn_skirt_back_04_l [{shapeType=Capsule, length=13.5, radius=3.5, shapeName=dyn_skirt_back_04_l_capsule, center=<4 0 0>, rotation=<rot 0 0 90>}]
 dyn_skirt_back_04_r [{shapeType=Capsule, length=13.5, radius=3.5, shapeName=dyn_skirt_back_04_r_capsule, center=<4 0 0>, rotation=<rot 0 0 90>}]
 dyn_skirt_front_01_l [{shapeType=Capsule, length=17, radius=3.5, shapeName=dyn_skirt_back_01_r_capsule, center=<4 0 0>, rotation=<rot 0 0 90>}]
 dyn_skirt_front_01_r [{shapeType=Capsule, length=17, radius=3.5, shapeName=dyn_skirt_back_01_r_capsule, center=<4 0 0>, rotation=<rot 0 0 90>}]
 dyn_skirt_front_02_l [{shapeType=Capsule, length=17, radius=3.5, shapeName=dyn_skirt_back_02_r_capsule, center=<4 0 0>, rotation=<rot 0 0 90>}]
 dyn_skirt_front_02_r [{shapeType=Capsule, length=17, radius=3.5, shapeName=dyn_skirt_back_02_r_capsule, center=<4 0 0>, rotation=<rot 0 0 90>}]
 dyn_skirt_front_03_l [{shapeType=Capsule, length=17, radius=3.5, shapeName=dyn_skirt_back_03_r_capsule, center=<4 0 0>, rotation=<rot 0 0 90>}]
 dyn_skirt_front_03_r [{shapeType=Capsule, length=17, radius=3.5, shapeName=dyn_skirt_back_03_r_capsule, center=<4 0 0>, rotation=<rot 0 0 90>}]
 foot_l [{shapeType=Capsule, length=6.3813, radius=10.2016, shapeName=foot_l_capsule}]
 foot_r [{shapeType=Capsule, length=6.3813, radius=10.2016, shapeName=foot_r_capsule}]
 pelvis [
  {shapeType=Capsule, length=10.2374, radius=9.6205, center=<-3.1833 -1.5286 0>}
  {shapeType=Capsule, length=11.25, radius=8.6875, shapeName=pelvis_capsule, center=<-5.1189 -3.2248 0>}
 ]
 root [{shapeType=Box, extentX=100, extentY=100, extentZ=2.5, shapeName=root_box}]
 thigh_l [{shapeType=Capsule, length=25, radius=9, center=<-18.002 0 0>, rotation=<rot 90 0 0>}]
 thigh_r [{shapeType=Capsule, length=25, radius=8, center=<18.002 0 0>, rotation=<rot 90 0 0>}]
}
constraints [
 {
  bone1Name=thigh_l
  bone2Name=pelvis
  swing1LimitDegrees=45
  swing1Motion=Free
  swing2LimitDegrees=45
  swing2Motion=Free
  twistLimitDegrees=45
  twistMotion=Free
 }
 {
  bone1Name=calf_l
  bone2Name=thigh_l
  swing1LimitDegrees=45
  swing1Motion=Free
  swing2LimitDegrees=45
  swing2Motion=Free
  twistLimitDegrees=45
  twistMotion=Free
 }
 {
  bone1Name=thigh_r
  bone2Name=pelvis
  swing1LimitDegrees=45
  swing1Motion=Free
  swing2LimitDegrees=45
  swing2Motion=Free
  twistLimitDegrees=45
  twistMotion=Free
 }
 {
  bone1Name=calf_r
  bone2Name=thigh_r
  swing1LimitDegrees=45
  swing1Motion=Free
  swing2LimitDegrees=45
  swing2Motion=Free
  twistLimitDegrees=45
  twistMotion=Free
 }
 {bone1Name=dyn_coat_back_01_l, bone2Name=pelvis, swing1LimitDegrees=20, swing2LimitDegrees=20, twistLimitDegrees=5}
 {bone1Name=dyn_coat_back_02_l, bone2Name=dyn_coat_back_01_l, swing1LimitDegrees=20, swing2LimitDegrees=20, twistLimitDegrees=5}
 {bone1Name=dyn_coat_back_03_l, bone2Name=dyn_coat_back_02_l, swing1LimitDegrees=20, swing2LimitDegrees=20, twistLimitDegrees=5}
 {bone1Name=dyn_coat_back_01_r, bone2Name=pelvis, swing1LimitDegrees=20, swing2LimitDegrees=20, twistLimitDegrees=5}
 {bone1Name=dyn_coat_back_02_r, bone2Name=dyn_coat_back_01_r, swing1LimitDegrees=20, swing2LimitDegrees=20, twistLimitDegrees=5}
 {bone1Name=dyn_coat_back_03_r, bone2Name=dyn_coat_back_02_r, swing1LimitDegrees=20, swing2LimitDegrees=20, twistLimitDegrees=5}
 {bone1Name=dyn_coat_front_01_l, bone2Name=pelvis, swing1LimitDegrees=20, swing2LimitDegrees=20, twistLimitDegrees=5}
 {bone1Name=dyn_coat_front_02_l, bone2Name=dyn_coat_front_01_l, swing1LimitDegrees=20, swing2LimitDegrees=20, twistLimitDegrees=5}
 {bone1Name=dyn_coat_front_03_l, bone2Name=dyn_coat_front_02_l, swing1LimitDegrees=20, swing2LimitDegrees=20, twistLimitDegrees=5}
 {bone1Name=dyn_coat_front_04_l, bone2Name=dyn_coat_front_03_l, swing1LimitDegrees=20, swing2LimitDegrees=20, twistLimitDegrees=5}
 {bone1Name=dyn_coat_front_01_r, bone2Name=pelvis, swing1LimitDegrees=20, swing2LimitDegrees=20, twistLimitDegrees=5}
 {bone1Name=dyn_coat_front_02_r, bone2Name=dyn_coat_front_01_r, swing1LimitDegrees=20, swing2LimitDegrees=20, twistLimitDegrees=5}
 {bone1Name=dyn_coat_front_03_r, bone2Name=dyn_coat_front_02_r, swing1LimitDegrees=20, swing2LimitDegrees=20, twistLimitDegrees=5}
 {bone1Name=dyn_coat_front_04_r, bone2Name=dyn_coat_front_03_r, swing1LimitDegrees=20, swing2LimitDegrees=20, twistLimitDegrees=5}
 {bone1Name=dyn_coat_mid_01_l, bone2Name=pelvis, swing1LimitDegrees=10, swing2LimitDegrees=20, twistLimitDegrees=5}
 {bone1Name=dyn_coat_mid_02_l, bone2Name=dyn_coat_mid_01_l, swing1LimitDegrees=10, swing2LimitDegrees=20, twistLimitDegrees=5}
 {
  bone1Name=dyn_coat_mid_03_l
  bone2Name=dyn_coat_mid_02_l
  swing1LimitDegrees=10
  swing2LimitDegrees=20
  twistLimitDegrees=5
  twistMotion=Locked
 }
 {bone1Name=dyn_coat_mid_01_r, bone2Name=pelvis, swing1LimitDegrees=20, swing2LimitDegrees=10, twistLimitDegrees=5}
 {bone1Name=dyn_coat_mid_02_r, bone2Name=dyn_coat_mid_01_r, swing1LimitDegrees=20, swing2LimitDegrees=10, twistLimitDegrees=5}
 {
  bone1Name=dyn_coat_mid_03_r
  bone2Name=dyn_coat_mid_02_r
  swing1LimitDegrees=20
  swing2LimitDegrees=10
  twistLimitDegrees=5
  twistMotion=Locked
 }
 {bone1Name=dyn_skirt_back_01_l, bone2Name=pelvis, swing1LimitDegrees=6, swing2LimitDegrees=12, twistLimitDegrees=2}
 {
  bone1Name=dyn_skirt_back_02_l
  bone2Name=dyn_skirt_back_01_l
  swing1LimitDegrees=6
  swing1Motion=Locked
  swing2LimitDegrees=12
  twistLimitDegrees=2
 }
 {
  bone1Name=dyn_skirt_back_03_l
  bone2Name=dyn_skirt_back_02_l
  swing1LimitDegrees=6
  swing1Motion=Locked
  swing2LimitDegrees=12
  twistLimitDegrees=2
 }
 {
  bone1Name=dyn_skirt_back_04_l
  bone2Name=dyn_skirt_back_03_l
  swing1LimitDegrees=6
  swing1Motion=Locked
  swing2LimitDegrees=12
  twistLimitDegrees=2
 }
 {bone1Name=dyn_skirt_back_01_r, bone2Name=pelvis, swing1LimitDegrees=6, swing2LimitDegrees=12, twistLimitDegrees=2}
 {
  bone1Name=dyn_skirt_back_02_r
  bone2Name=dyn_skirt_back_01_r
  swing1LimitDegrees=6
  swing1Motion=Locked
  swing2LimitDegrees=12
  twistLimitDegrees=2
 }
 {
  bone1Name=dyn_skirt_back_03_r
  bone2Name=dyn_skirt_back_02_r
  swing1LimitDegrees=6
  swing1Motion=Locked
  swing2LimitDegrees=12
  twistLimitDegrees=2
 }
 {
  bone1Name=dyn_skirt_back_04_r
  bone2Name=dyn_skirt_back_03_r
  swing1LimitDegrees=6
  swing1Motion=Locked
  swing2LimitDegrees=12
  twistLimitDegrees=2
 }
 {
  bone1Name=foot_r
  bone2Name=calf_r
  swing1LimitDegrees=45
  swing1Motion=Free
  swing2LimitDegrees=45
  swing2Motion=Free
  twistLimitDegrees=45
  twistMotion=Free
 }
 {
  bone1Name=foot_l
  bone2Name=calf_l
  swing1LimitDegrees=45
  swing1Motion=Free
  swing2LimitDegrees=45
  swing2Motion=Free
  twistLimitDegrees=45
  twistMotion=Free
 }
 {
  bone1Name=dyn_skirt_back_04_l
  bone2Name=dyn_skirt_back_04_r
  swing1LimitDegrees=45
  swing1Motion=Free
  swing2LimitDegrees=45
  swing2Motion=Free
  twistLimitDegrees=45
  twistMotion=Free
 }
 {bone1Name=dyn_skirt_front_01_l, bone2Name=thigh_l, swing1LimitDegrees=3, swing2LimitDegrees=8, twistLimitDegrees=5}
 {bone1Name=dyn_skirt_front_02_l, bone2Name=dyn_skirt_front_01_l, swing1LimitDegrees=3, swing2LimitDegrees=30, twistLimitDegrees=5}
 {bone1Name=dyn_skirt_front_03_l, bone2Name=dyn_skirt_front_02_l, swing1LimitDegrees=15, swing2LimitDegrees=30, twistLimitDegrees=5}
 {bone1Name=dyn_skirt_front_01_r, bone2Name=thigh_r, swing1LimitDegrees=3, swing2LimitDegrees=8, twistLimitDegrees=5}
 {bone1Name=dyn_skirt_front_02_r, bone2Name=dyn_skirt_front_01_r, swing1LimitDegrees=3, swing2LimitDegrees=8, twistLimitDegrees=5}
 {bone1Name=dyn_skirt_front_03_r, bone2Name=dyn_skirt_front_02_r, swing1LimitDegrees=15, swing2LimitDegrees=30, twistLimitDegrees=5}
 {
  bone1Name=pelvis
  bone2Name=root
  swing1LimitDegrees=45
  swing1Motion=Free
  swing2LimitDegrees=45
  swing2Motion=Free
  twistLimitDegrees=45
  twistMotion=Free
 }
 {
  bone1Name=dyn_coat_mid_02_r
  bone2Name=dyn_coat_front_02_r
  swing1LimitDegrees=45
  swing1Motion=Free
  swing2LimitDegrees=45
  swing2Motion=Free
  twistLimitDegrees=45
  twistMotion=Free
 }
 {
  bone1Name=dyn_coat_mid_02_l
  bone2Name=dyn_coat_front_02_l
  swing1LimitDegrees=45
  swing1Motion=Free
  swing2LimitDegrees=45
  swing2Motion=Free
  twistLimitDegrees=45
  twistMotion=Free
 }
]
