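# Post-pose secondary-motion AnimBlueprint for the LazarusLens Style NPC body, copying its pose from the attached parent mesh.
# A RigidBody node simulates physics in pelvis space, then additive bone rotations lift back skirt and coat bones, and AnimDynamics adds motion to two bun bones.
# Each update maps calf-versus-thigh yaw delta from ref pose to clamped alphas driving those rotations.
# Skirt bones follow same-side legs, while coat bones follow opposite legs.
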AnimBlueprint /BRCosmetics/Characters/Player/Male/Medium/Bodies/M_MED_LazarusLens_StyleNPC/Meshes/M_MED_LazarusLens_Style_NPC_AnimBP
target_skeleton=&/M_MED_LazarusLens_Style_NPC_Skeleton
graphs [
 AnimGraph {nodes [
   n1 = RigidBody(ComponentPose=Local To Component(LocalPose=Copy Pose From Mesh() {bUseAttachedParent=true})) {
    bOverrideWorldGravity=true
    overridePhysicsAsset=&/M_MED_LazarusLens_Style_NPC_Physics
    simulationSpace=BaseBoneSpace
    alphaBoolBlend {blendOption=Linear}
    baseBoneRef {boneName=pelvis}
    componentAppliedLinearAccClamp=<3000 3000 3000>
    componentLinearAccScale=<0.3 0.3 0.3>
    componentLinearVelScale=<0.3 0.3 0.3>
    overrideWorldGravity {z=-980}
   }
   n2 = Get calf_l_alpha() -> (calf_l_alpha:double)
   n3 = "Transform (Modify) Bone - Bone: dyn_skirt_back_master_l"(ComponentPose=@n1, Alpha=@n2, Translation="0,0,0", Rotation="10,0,0", Scale="1,1,1") {
    rotationMode=BMM_Additive
    rotationSpace=BCS_BoneSpace
    alphaBoolBlend {blendOption=Linear}
    boneToModify {boneName=dyn_skirt_back_master_l}
    rotation {pitch=10}
   }
   n4 = Get calf_r_alpha() -> (calf_r_alpha:double)
   n5 = "Transform (Modify) Bone - Bone: dyn_skirt_back_master_r"(ComponentPose=@n3, Alpha=@n4, Translation="0,0,0", Rotation="10,0,0", Scale="1,1,1") {
    rotationMode=BMM_Additive
    rotationSpace=BCS_BoneSpace
    alphaBoolBlend {blendOption=Linear}
    boneToModify {boneName=dyn_skirt_back_master_r}
    rotation {pitch=10}
   }
   n6 = "Transform (Modify) Bone - Bone: dyn_coat_back_01_r"(ComponentPose=@n5, Alpha=@n2, Translation="0,0,0", Rotation="25,0,0", Scale="1,1,1") {
    rotationMode=BMM_Additive
    rotationSpace=BCS_BoneSpace
    alphaBoolBlend {blendOption=Linear}
    boneToModify {boneName=dyn_coat_back_01_r}
    rotation {pitch=25}
   }
   n7 = "Transform (Modify) Bone - Bone: dyn_coat_back_01_l"(ComponentPose=@n6, Alpha=@n4, Translation="0,0,0", Rotation="25,0,0", Scale="1,1,1") {
    rotationMode=BMM_Additive
    rotationSpace=BCS_BoneSpace
    alphaBoolBlend {blendOption=Linear}
    boneToModify {boneName=dyn_coat_back_01_l}
    rotation {pitch=25}
   }
   n8 = Anim Dynamics - Bone: dyn_bun_01(ComponentPose=@n7, Alpha=0.7, ExternalForce="0,0,0") {
    angularDampingOverride=0.9
    angularSpringConstant=20000
    bOverrideAngularDamping=true
    bUseGravityOverride=true
    simulationSpace=RootRelative
    alphaBoolBlend {blendOption=Linear}
    boundBone {boneName=dyn_bun_01}
    componentAppliedLinearAccClamp=<2000 2000 3000>
    componentLinearAccScale=<0.15 0.15 0.2>
    componentLinearVelScale=<0.15 0.15 0.2>
    gravityOverride {z=-980}
    physicsBodyDefinitions [{
      sphereCollisionRadius=0
      boundBone {boneName=dyn_bun_01}
      boxExtents=<15 12 12>
      constraintSetup {angularLimitsMax=<10 10 10>, angularLimitsMin=<-5 -10 -3>, angularTarget=<1 0 0>}
      localJointOffset=<7 0 0>
     }]
    relativeSpaceBone {boneName=head}
   }
   n9 = Anim Dynamics - Bone: dyn_bun_02(ComponentPose=@n8, Alpha=0.7, ExternalForce="0,0,0") {
    angularDampingOverride=0.9
    angularSpringConstant=20000
    bOverrideAngularDamping=true
    bUseGravityOverride=true
    simulationSpace=RootRelative
    alphaBoolBlend {blendOption=Linear}
    boundBone {boneName=dyn_bun_02}
    componentAppliedLinearAccClamp=<1500 1500 2500>
    componentLinearAccScale=<0.2 0.2 0.3>
    componentLinearVelScale=<0.2 0.2 0.3>
    gravityOverride {z=-980}
    physicsBodyDefinitions [{
      sphereCollisionRadius=0
      boundBone {boneName=dyn_bun_02}
      boxExtents=<6 9 9>
      constraintSetup {angularLimitsMax=<15 15 15>, angularLimitsMin=<-15 -15 -15>, angularTarget=<1 0 0>}
      localJointOffset=<3 0 0>
     }]
    relativeSpaceBone {boneName=head}
   }
   Output Pose(Result=Component To Local(ComponentPose=@n9)) {layerGroup=None}
  ]}
 EventGraph {nodes [
   Event BlueprintUpdateAnimation()
   n1 = GetOwnerSkeletalMeshComponent()
   n2 = BreakTransform(InTransform=GetDeltaTransformFromRefPose(self=@n1, BoneName=calf_r, BaseName=thigh_r))
   > Set calf_r_alpha(calf_r_alpha:double=MapRangeClamped(Value=BreakRotator(InRot=@n2.Rotation).Yaw, InRangeA=0, InRangeB=90)) -> (Output_Get:double)
   n3 = BreakTransform(InTransform=GetDeltaTransformFromRefPose(self=@n1, BoneName=calf_l, BaseName=thigh_l))
   > Set calf_l_alpha(calf_l_alpha:double=MapRangeClamped(Value=BreakRotator(InRot=@n3.Rotation).Yaw, InRangeA=0, InRangeB=90)) -> (Output_Get:double)
   TryGetPawnOwner()
  ]}
]
variables [
 calf_l_alpha {type=number}
 calf_r_alpha {type=number}
]
